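# Unlit translucent hologram material for the dealership globe, flagged for Niagara mesh particles and tinted and faded by particle color.
# An infinite-offset globe texture, Fresnel terms and two hex pattern layers, one panning, feed emissive and opacity.
# The UseHexagonGrid? scalar switches between hex-grid and plain texture looks through If nodes rather than a static switch.
# Time-driven sine and cosine waves build a normal-based world position offset, scaled by the WPOAmount dynamic parameter.
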Material /DealershipSharedLibrary/Assets/Sets/Globe/M_DLR_Globe_Hologram
graph {
 properties {
  bAutomaticallySetUsageInEditor=false
  bUsedWithNiagaraMeshParticles=true
  blendMode=BLEND_Translucent
  shadingModel=MSM_Unlit
  usedShadingModels=Unlit
 }
 nodes [
  n1 = Saturate(None=NamedRerouteUsage())
  n2 = Fresnel(ExponentIn=Multiply(A=Add(A=@n1) {constB=0.5}, B=ScalarParameter() {defaultValue=1, parameterName=Param_1})) {baseReflectFraction=0.1, exponent=1}
  n3 = ParticleColor()
  n4 = Multiply(A=Multiply(A=Multiply(A=@n2) {constB=2}, B=NamedRerouteUsage()), B=@n3.A)
  If(
   A=ScalarParameter() {defaultValue=1, parameterName=UseHexagonGrid?}
   "A > B"=@n4
   "A == B"=@n4
   "A < B"=Multiply(A=NamedRerouteUsage(), B=@n3.A)
  ) {constB=1}
  n5 = ParticleColor()
  n6 = Multiply(A=Max(A=NamedRerouteUsage(), B=NamedRerouteUsage()), B=NamedRerouteUsage())
  n7 = Multiply(A=@n5.RGB, B=Multiply(A=LinearInterpolate(Alpha=@n6) {constA=0.2, constB=0.5}, B=Add(A=@n2) {constB=0.5}))
  MaterialFunctionCall(Emissive=If(A=ScalarParameter() {defaultValue=1, parameterName=UseHexagonGrid?}, "A > B"=@n7, "A == B"=@n7, "A < B"=NamedRerouteUsage()) {constB=1}) {materialFunction=&MF_FX_EmissiveAdjustments}
  n8 = VectorParameter() {parameterName=InfiniteUVControl, channelNames {a=SpeedY, b=SpeedX, g=TilingY, r=TilingX}, defaultValue=<rgba 1 1 0 0>}
  n9 = TextureCoordinate()
  NamedRerouteDeclaration(None=MaterialFunctionCall!(
    bRotateUVs=StaticBool()
    bGradientFalloff=StaticBool() {value=true}
    GradientDarkness=ScalarParameter() {defaultValue=0.05, parameterName=GradientDarkness}
    Offset=ScalarParameter() {defaultValue=0.444, parameterName=Offset}
    UVs=@n8.RGBA
    Texture=TextureObjectParameter(UVs=@n9) {parameterName=MainTexture, texture=&/T_Globe}
   ) {materialFunction=/DealershipSharedLibrary/Materials/MaterialFunctions/MF_DLR_InfinityOffset}.Quality Infinite Offset) {name=InfinityOffset, nodeColor=<rgba 0.2188 0 1 1>}
  NamedRerouteDeclaration(None=Fresnel(ExponentIn=ScalarParameter() {parameterName=FresnelExponent}) {baseReflectFraction=0}) {name=BasicFresnel, nodeColor=<rgba 1 0 0.375 1>}
  n10 = NamedRerouteUsage()
  NamedRerouteDeclaration(None=Multiply(A=NamedRerouteUsage(), B=Multiply(A=@n10, B=@n5.RGB))) {name=TextureNoGridOut, nodeColor=<rgba 1 0.8438 0 1>}
  n11 = Add(
   A=Multiply(A=Time(), B=ScalarParameter() {defaultValue=1, parameterName=WPOTimeMultiplierZ})
   B=ComponentMask(None=TextureCoordinate()) {g=true}
  )
  n12 = Cosine(None=@n11) {period=0.8}
  n13 = AppendVector(A=Sine(None=@n11) {period=1.5}, B=@n12)
  n14 = Multiply(A=ConstantBiasScale(None=AppendVector(A=@n13, B=@n12)) {scale=2}, B=VertexNormalWS())
  n15 = Add(
   A=Multiply(A=Time(), B=ScalarParameter() {defaultValue=1, parameterName=WPOTimeMultiplierXY})
   B=ComponentMask(None=TextureCoordinate()) {r=true}
  )
  n16 = Cosine(None=@n15)
  n17 = AppendVector(A=AppendVector(A=Sine(None=@n15) {period=2}, B=@n16), B=@n16)
  n18 = Multiply(A=Multiply(A=ConstantBiasScale(None=@n17) {scale=1.5}, B=VertexNormalWS()), B=ScalarParameter() {parameterName=WPO_XYMultipler})
  NamedRerouteDeclaration(None=Multiply(A=Add(A=@n14, B=@n18), B=DynamicParameter() {defaultValue=<rgba 0 1 1 1>, paramNames [WPOAmount,Param2,Param3,Param4]}.WPOAmount)) {name=WPO, nodeColor=<rgba 1 0 0.0154 1>}
  n19 = Multiply(A=TextureCoordinate() {uTiling=2}, B=ScalarParameter() {defaultValue=8, parameterName=Param_3})
  NamedRerouteDeclaration(None=Multiply(A=TextureSample(UVs=@n19) {samplerType=SAMPLERTYPE_LinearGrayscale, texture=/DealershipSharedLibrary/FX/Textures/Masks/T_Hex_Pattern}.R) {constB=0.8}) {name=HexLargeOut, nodeColor=<rgba 0 1 0.2031 1>}
  n20 = Panner(Coordinate=Multiply(A=TextureCoordinate(), B=ScalarParameter() {defaultValue=30, parameterName=Param_2})) {speedY=1}
  NamedRerouteDeclaration(None=LinearInterpolate(Alpha=TextureSample(UVs=@n20) {samplerType=SAMPLERTYPE_LinearColor, texture=/Game/Global/Textures/Default/Masks_and_Gradients/HexPattern}.R) {constA=0.8}) {name=HexSmallOut, nodeColor=<rgba 1 0.9392 0.2504 1>}
  NamedRerouteUsage()
 ]
}
